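# Physics asset for the Halloween Bunny female medium Commando body cosmetic, giving its tail and small "light" pieces at shoulders, hips and left thigh physical motion.
# All bodies are capsules, with pelvis, spine, clavicles and left thigh bones forming a freely jointed anchor chain.
# Dangling pieces hang off clavicles, pelvis or dyn_thigh_l with limited swing and tight twist; the tail instead twists widely.
# Two shoulder lights are constrained to the pelvis rather than a clavicle.
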PhysicsAsset /BRCosmetics/Characters/Player/Female/Medium/Bodies/F_MED_Commando_HalloweenBunny/Meshes/F_MED_Commando_HalloweenBunny_Body_Physics
bodies {
 clavicle_l [{shapeType=Capsule, length=4.1043, radius=1.8264, center=<6.2726 -2.6137 -1.246>, rotation=<rot 0 176.6103 -68.6472>}]
 clavicle_r [{shapeType=Capsule, length=4.8835, radius=1.5525, center=<-6.1918 2.55 1.7293>, rotation=<rot 0 -179.5661 -62.1582>}]
 dyn_hip_light_02 [{shapeType=Capsule, length=2.8552, radius=1.4031, center=<4.2638 0.0017 0.003>, rotation=<rot 0 90.0578 -89.908>}]
 dyn_hip_light_03 [{shapeType=Capsule, length=4.3551, radius=1.4108, center=<4.6122 0.0028 -0.0012>, rotation=<rot 0 -89.8966 -89.9665>}]
 dyn_hip_light_04 [{shapeType=Capsule, length=2.8552, radius=1.445, center=<2.7101 0.0006 0>, rotation=<rot 0 89.8913 -89.9971>}]
 dyn_hip_light_05 [{shapeType=Capsule, length=2.8552, radius=1.4767, center=<-3.4204 -0.0084 0.083>, rotation=<rot 0 -89.801 -88.5591>}]
 dyn_hip_light_06 [{shapeType=Capsule, length=2.8552, radius=1.4767, center=<-2.7093 -0.0059 0.0653>, rotation=<rot 0 -89.8011 -88.5578>}]
 dyn_hip_light_08 [{shapeType=Capsule, length=3.3552, radius=2.9768, center=<-3.7723 -0.0094 0.0923>, rotation=<rot 0 -89.8043 -88.5538>}]
 dyn_hip_light_09 [{shapeType=Capsule, length=2.8552, radius=1.4767, center=<-4.1575 -0.0107 0.102>, rotation=<rot 0 -89.8034 -88.5543>}]
 dyn_hip_light_10 [{shapeType=Capsule, length=2.8552, radius=1.4767, center=<-2.7092 -0.0059 0.0653>, rotation=<rot 0 -89.8014 -88.5575>}]
 dyn_shldr_light_01 [{shapeType=Capsule, length=2.8552, radius=1.4767, center=<-3.708 -0.0094 0.0908>, rotation=<rot 0 -89.8009 -88.5593>}]
 dyn_shldr_light_02 [{shapeType=Capsule, length=2.8552, radius=1.4767, center=<-2.7092 -0.0059 0.0652>, rotation=<rot 0 -89.8016 -88.5587>}]
 dyn_shldr_light_03 [{shapeType=Capsule, length=2.8552, radius=1.4767, center=<-2.7093 -0.0059 0.0653>, rotation=<rot 0 -89.801 -88.5573>}]
 dyn_shldr_light_04 [{shapeType=Capsule, length=2.8551, radius=1.4767, center=<-2.7092 -0.0058 0.0652>, rotation=<rot 0 -89.8023 -88.5584>}]
 dyn_shldr_light_05 [{shapeType=Capsule, length=2.8552, radius=1.4767, center=<-3.8524 -0.0097 0.0941>, rotation=<rot 0 -89.8027 -88.5572>}]
 dyn_shldr_light_06 [{shapeType=Capsule, length=2.8552, radius=1.4767, center=<-2.7093 -0.0058 0.0652>, rotation=<rot 0 -89.8018 -88.5582>}]
 dyn_shldr_light_07 [{shapeType=Capsule, length=2.8552, radius=1.4767, center=<-3.6187 -0.009 0.0882>, rotation=<rot 0 -89.8009 -88.557>}]
 dyn_tail [{shapeType=Capsule, length=7.8736, radius=6.4647, center=<0.0002 -5.4953 1.4518>, rotation=<rot 0 0 109.7504>}]
 dyn_thigh_l [{shapeType=Capsule, length=1.5404, radius=0.8519, center=<-0.5235 -0.1324 0.0057>, rotation=<rot 0 -154.9198 -87.8502>}]
 dyn_thigh_light_01 [{shapeType=Capsule, length=2.8552, radius=1.3639, center=<4.714 0.3908 -0.141>, rotation=<rot 0 -85.1974 -88.2682>}]
 dyn_thigh_light_02 [{shapeType=Capsule, length=3.0173, radius=3.0657, center=<-3.9992 0.0002 0>, rotation=<rot 0 90 90>}]
 pelvis [{shapeType=Capsule, length=2.9248, radius=5.7942}]
 spine_05 [{shapeType=Capsule, length=1.0199, radius=4.1087}]
 thigh_l [{shapeType=Capsule, length=2.923, radius=1.5, center=<-32.1125 1.9106 -1.0463>, rotation=<rot 0 85.3885 -89.0944>}]
 thigh_twist_01_l [{shapeType=Capsule, length=4.5943, radius=2.3663, center=<18.5834 -0.2784 -0.1735>, rotation=<rot -33.5288 -90 -18.721>}]
}
constraints [
 {
  bone1Name=spine_05
  bone2Name=pelvis
  swing1LimitDegrees=45
  swing1Motion=Free
  swing2LimitDegrees=45
  swing2Motion=Free
  twistLimitDegrees=45
  twistMotion=Free
 }
 {bone1Name=dyn_shldr_light_01, bone2Name=clavicle_l, swing1LimitDegrees=20, swing2LimitDegrees=20, twistLimitDegrees=5}
 {bone1Name=dyn_shldr_light_02, bone2Name=clavicle_l, swing1LimitDegrees=30, swing2LimitDegrees=25, twistLimitDegrees=5}
 {bone1Name=dyn_shldr_light_03, bone2Name=clavicle_l, swing1LimitDegrees=40, swing2LimitDegrees=35, twistLimitDegrees=5}
 {bone1Name=dyn_shldr_light_06, bone2Name=clavicle_r, swing1LimitDegrees=20, swing2LimitDegrees=20, twistLimitDegrees=5}
 {bone1Name=dyn_shldr_light_05, bone2Name=clavicle_r, swing1LimitDegrees=20, swing2LimitDegrees=20, twistLimitDegrees=5}
 {bone1Name=dyn_shldr_light_04, bone2Name=pelvis, swing1LimitDegrees=20, swing2LimitDegrees=20, twistLimitDegrees=5}
 {bone1Name=dyn_shldr_light_07, bone2Name=pelvis, swing1LimitDegrees=20, swing2LimitDegrees=20, twistLimitDegrees=5}
 {bone1Name=dyn_hip_light_02, bone2Name=pelvis, swing1LimitDegrees=20, swing2LimitDegrees=20, twistLimitDegrees=5}
 {bone1Name=dyn_hip_light_03, bone2Name=pelvis, swing1LimitDegrees=20, swing2LimitDegrees=20, twistLimitDegrees=5}
 {bone1Name=dyn_hip_light_04, bone2Name=pelvis, swing1LimitDegrees=40, swing2LimitDegrees=35, twistLimitDegrees=5}
 {bone1Name=dyn_hip_light_05, bone2Name=pelvis, swing1LimitDegrees=30, swing2LimitDegrees=30, twistLimitDegrees=5}
 {bone1Name=dyn_hip_light_06, bone2Name=pelvis, swing1LimitDegrees=35, swing2LimitDegrees=35, twistLimitDegrees=5}
 {bone1Name=dyn_hip_light_10, bone2Name=pelvis, swing1LimitDegrees=20, swing2LimitDegrees=20, twistLimitDegrees=5}
 {bone1Name=dyn_hip_light_09, bone2Name=pelvis, swing1LimitDegrees=20, swing2LimitDegrees=20, twistLimitDegrees=5}
 {bone1Name=dyn_hip_light_08, bone2Name=pelvis, swing1LimitDegrees=35, swing2LimitDegrees=30, twistLimitDegrees=5}
 {bone1Name=dyn_tail, bone2Name=pelvis, swing1LimitDegrees=60, swing2LimitDegrees=5, twistLimitDegrees=75}
 {bone1Name=dyn_thigh_light_01, bone2Name=dyn_thigh_l, swing1LimitDegrees=35, swing2LimitDegrees=40, twistLimitDegrees=5}
 {bone1Name=dyn_thigh_light_02, bone2Name=dyn_thigh_l, swing1LimitDegrees=40, swing2LimitDegrees=40, twistLimitDegrees=5}
 {
  bone1Name=clavicle_r
  bone2Name=spine_05
  swing1LimitDegrees=45
  swing1Motion=Free
  swing2LimitDegrees=45
  swing2Motion=Free
  twistLimitDegrees=45
  twistMotion=Free
 }
 {
  bone1Name=clavicle_l
  bone2Name=spine_05
  swing1LimitDegrees=45
  swing1Motion=Free
  swing2LimitDegrees=45
  swing2Motion=Free
  twistLimitDegrees=45
  twistMotion=Free
 }
 {
  bone1Name=thigh_l
  bone2Name=pelvis
  swing1LimitDegrees=45
  swing1Motion=Free
  swing2LimitDegrees=45
  swing2Motion=Free
  twistLimitDegrees=45
  twistMotion=Free
 }
 {
  bone1Name=thigh_twist_01_l
  bone2Name=thigh_l
  swing1LimitDegrees=45
  swing1Motion=Free
  swing2LimitDegrees=45
  swing2Motion=Free
  twistLimitDegrees=45
  twistMotion=Free
 }
 {
  bone1Name=dyn_thigh_l
  bone2Name=thigh_l
  swing1LimitDegrees=45
  swing1Motion=Free
  swing2LimitDegrees=45
  swing2Motion=Free
  twistLimitDegrees=45
  twistMotion=Free
 }
]
