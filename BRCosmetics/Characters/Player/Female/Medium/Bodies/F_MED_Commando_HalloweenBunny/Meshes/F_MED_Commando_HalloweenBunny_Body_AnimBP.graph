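# Post-process style body AnimBP for the Halloween Bunny female Commando skin, adding secondary motion and corrective bone placement to a pose copied from the attached parent mesh.
# AnimDynamics moves bandana, tags and boot strings, then a crouch-driven thigh offset feeds a RigidBody simulation.
# Weighted constraints position shoulder and hip light bones; two single-target shoulder constraints never reach the output.
# The update event only sets crouch_alpha; cape/backbling detection and material-hiding nodes are not connected to it.
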AnimBlueprint /BRCosmetics/Characters/Player/Female/Medium/Bodies/F_MED_Commando_HalloweenBunny/Meshes/F_MED_Commando_HalloweenBunny_Body_AnimBP
target_skeleton=&/F_MED_Commando_HalloweenBunny_Body_Skeleton
graphs [
 AnimGraph {nodes [
   n1 = "Constraint - dyn_shldr_light_05 (P)"!(ComponentPose, Alpha=1, ConstraintWeights_0=1, ConstraintWeights_1:float=0) -> (Pose) {
    alphaBoolBlend {blendOption=Linear}
    boneToModify {boneName=dyn_shldr_light_05}
    constraintSetup [{targetBone {boneName=clavicle_r}}]
    constraintWeights [1,0]
   }
   "Constraint - dyn_shldr_light_01 (P)"!(ComponentPose=@n1, Alpha=1, ConstraintWeights_0=1, ConstraintWeights_1:float=0) -> (Pose) {
    alphaBoolBlend {blendOption=Linear}
    boneToModify {boneName=dyn_shldr_light_01}
    constraintSetup [{targetBone {boneName=clavicle_l}}]
    constraintWeights [1,0]
   }
   n2 = Anim Dynamics - Bone: dyn_arm_r_bandana(
    ComponentPose=Local To Component(LocalPose=Copy Pose From Mesh() {bCopyCurves=true, bUseAttachedParent=true})
    Alpha=0.75
    ExternalForce="0,0,0"
   ) {
    alpha=0.9
    angularDampingOverride=0.985
    angularSpringConstant=300
    bAngularSpring=true
    bGravityOverrideInSimSpace=true
    bOverrideAngularDamping=true
    gravityScale=0.3
    numSolverIterationsPostUpdate=4
    numSolverIterationsPreUpdate=16
    alphaBoolBlend {blendOption=Linear}
    boundBone {boneName=dyn_arm_r_bandana}
    physicsBodyDefinitions [{
      sphereCollisionRadius=0
      boundBone {boneName=dyn_arm_r_bandana}
      boxExtents=<20 8 8>
      constraintSetup {angularTargetAxis=AxisZ, twistAxis=AxisY, angularLimitsMax=<15 0 15>, angularLimitsMin=<-15 0 -15>, angularTarget=<0 0 1>}
      localJointOffset=<0 0 -3>
     }]
   }
   n3 = Anim Dynamics - Bone: dyn_tags(ComponentPose=@n2, Alpha=0.35, ExternalForce="0,0,0") {
    angularBiasOverride=1
    angularDampingOverride=0.98
    bGravityOverrideInSimSpace=true
    bOverrideAngularDamping=true
    gravityScale=0.3
    lODThreshold=1
    numSolverIterationsPostUpdate=4
    numSolverIterationsPreUpdate=16
    alphaBoolBlend {blendOption=Linear}
    boundBone {boneName=dyn_tags}
    physicsBodyDefinitions [{
      sphereCollisionRadius=0
      boundBone {boneName=dyn_tags}
      boxExtents=<20 2 20>
      constraintSetup {angularTargetAxis=AxisZ, twistAxis=AxisZ, angularLimitsMax=<25 45 0>, angularLimitsMin=<0 -45 0>, angularTarget=<0 0 0>}
      localJointOffset=<0 1 -5>
     }]
   }
   n4 = Anim Dynamics - Bone: dyn_boot_strings_01(ComponentPose=@n3, Alpha=0.55, ExternalForce="0,0,0") {
    bGravityOverrideInSimSpace=true
    bLinearSpring=true
    bOverrideLinearDamping=true
    gravityScale=0.33
    linearDampingOverride=0.995
    linearSpringConstant=150
    simulationSpace=RootRelative
    alphaBoolBlend {blendOption=Linear}
    boundBone {boneName=dyn_boot_strings_01}
    physicsBodyDefinitions [{
      sphereCollisionRadius=0
      boundBone {boneName=dyn_boot_strings_01}
      boxExtents=<10 10 10>
      constraintSetup {angularLimitsMax=<0 0 0>, angularLimitsMin=<0 0 0>, angularTarget=<0 0 0>, linearAxesMax=<4 2 0.02>, linearAxesMin=<-4 -2 -0.02>}
      localJointOffset=<0 0 -5>
     }]
   }
   n5 = Anim Dynamics - Bone: dyn_boot_strings_02(ComponentPose=@n4, Alpha=0.3, ExternalForce="0,0,0") {
    bGravityOverrideInSimSpace=true
    bLinearSpring=true
    bOverrideLinearDamping=true
    gravityScale=0.29
    linearDampingOverride=0.995
    linearSpringConstant=150
    simulationSpace=RootRelative
    alphaBoolBlend {blendOption=Linear}
    boundBone {boneName=dyn_boot_strings_02}
    physicsBodyDefinitions [{
      sphereCollisionRadius=0
      boundBone {boneName=dyn_boot_strings_02}
      boxExtents=<10 10 10>
      constraintSetup {angularLimitsMax=<0 0 0>, angularLimitsMin=<0 0 0>, angularTarget=<0 0 0>, linearAxesMax=<4 2 0.02>, linearAxesMin=<-5 -2 -0.02>}
      localJointOffset=<0 0 -6>
     }]
   }
   n6 = "Constraint - dyn_shldr_light_07 (PPP)"!(ComponentPose=@n5, Alpha=1, ConstraintWeights_0=0.2717, ConstraintWeights_1:float=0.4369, ConstraintWeights_2:float=0.2914) -> (Pose) {
    alphaBoolBlend {blendOption=Linear}
    boneToModify {boneName=dyn_shldr_light_07}
    constraintSetup [{targetBone {boneName=clavicle_r}}, {targetBone {boneName=spine_04}}, {targetBone {boneName=spine_05}}]
    constraintWeights [1,1,1]
   }
   n7 = "Transform (Modify) Bone - Bone: dyn_thigh_l"(ComponentPose=@n6, Alpha=Get crouch_alpha() -> (crouch_alpha:double), Translation="-0.5,0,0.75", Rotation="10,0,30", Scale="1,1,1") {
    rotationMode=BMM_Additive
    rotationSpace=BCS_BoneSpace
    translationMode=BMM_Additive
    translationSpace=BCS_BoneSpace
    alphaBoolBlend {blendOption=Linear}
    alphaScaleBiasClamp {bInterpResult=true}
    boneToModify {boneName=dyn_thigh_l}
    rotation {pitch=10, roll=30}
    translation {x=-0.5, z=0.75}
   }
   n8 = RigidBody(ComponentPose=@n7, Alpha=0.7) {
    bClampLinearTranslationLimitToRefPose=true
    bOverrideWorldGravity=true
    overridePhysicsAsset=&/F_MED_Commando_HalloweenBunny_Body_Physics
    simulationSpace=BaseBoneSpace
    alphaBoolBlend {blendOption=Linear}
    baseBoneRef {boneName=root}
    overrideWorldGravity {z=-980}
   }
   n9 = "Constraint - dyn_hip_light_08 (PP)"!(ComponentPose=@n8, Alpha=1, ConstraintWeights_0=0.8193, ConstraintWeights_1:float=0.1807) -> (Pose) {
    alphaBoolBlend {blendOption=Linear}
    boneToModify {boneName=dyn_hip_light_08}
    constraintSetup [{targetBone {boneName=pelvis}}, {targetBone {boneName=thigh_twist_01_l}}]
    constraintWeights [1,1]
   }
   n10 = "Constraint - dyn_shldr_light_07 (PPPP)"!(
    ComponentPose=@n9
    Alpha=1
    ConstraintWeights_0=0.253
    ConstraintWeights_1:float=0.327
    ConstraintWeights_2:float=0.2772
    ConstraintWeights_3:float=0.1429
   ) -> (Pose) {
    alphaBoolBlend {blendOption=Linear}
    boneToModify {boneName=dyn_shldr_light_07}
    constraintSetup [
     {targetBone {boneName=clavicle_r}}
     {targetBone {boneName=spine_04}}
     {targetBone {boneName=spine_05}}
     {targetBone {boneName=dyn_shldr_light_07}}
    ]
    constraintWeights [1,1,1,1]
   }
   n11 = "Constraint - dyn_shldr_light_04 (PPP)"!(ComponentPose=@n10, Alpha=1, ConstraintWeights_0=0.9237, ConstraintWeights_1:float=0.0571, ConstraintWeights_2:float=0.0191) -> (Pose) {
    alphaBoolBlend {blendOption=Linear}
    boneToModify {boneName=dyn_shldr_light_04}
    constraintSetup [{targetBone {boneName=spine_04}}, {targetBone {boneName=spine_05}}, {targetBone {boneName=clavicle_l}}]
    constraintWeights [1,1,1]
   }
   Output Pose(Result=Component To Local(ComponentPose=@n11)) {layerGroup=None}
  ]}
 EventGraph {nodes [
   Event BlueprintUpdateAnimation()
   > n1 = Branch(Condition=Get bIsCrouching())
   > Set crouch_alpha(crouch_alpha:double=1) -> (Output_Get:double)
   Set crouch_alpha(execute=@n1.else, crouch_alpha=0)
   Branch(Condition=Get isWearingCape() -> (isWearingCape:bool))
   n2 = GetPartSkeletalMeshComponent()
   > n3 = Get Dynamic Material Instance(SkeletalMeshComponent=@n2)
   > n4 = SetScalarParameterValue(self=@n3, ParameterName=Hide Element 01, Value=1)
   Do Once(execute=[@n4.then, @n8.OutputPin])
   n5 = FindChildCommponentFromAttachSocket!(
    self:Custom Character Part Anim Instance Object Reference
    Parent:Skeletal Mesh Component Object Reference=@n2
    bIncludeAllDescendants:bool=false
    AttachSocket:Name=Cape
   ) -> (ReturnValue:Skeletal Mesh Component Object Reference)
   n6 = "Not Equal (Object)"!(A:Object Reference=@n5, B:Object Reference) -> (ReturnValue:bool)
   n7 = FindChildCommponentFromAttachSocket(Parent=@n2, bIncludeAllDescendants=false, AttachSocket=backpack_br)
   > Set isWearingBackbling(isWearingBackbling=OR Boolean(A=@n6, B="Not Equal (Object)"(A=@n7)))
   > Branch(Condition=@n6)
   > Set isWearingCape(isWearingCape="float > float"(
     A=Get Distance Between Two Sockets!(
      Component:Skeletal Mesh Component Object Reference=@n5
      SocketOrBoneNameA:Name=backpack_root
      SocketSpaceA:ERelativeTransformSpace Enum=RTS_World
      SocketOrBoneNameB:Name=cape_length
      SocketSpaceB:ERelativeTransformSpace Enum=RTS_World
      bRemapRange:bool=false
      InRangeMin:float=0
      InRangeMax:float=0
      OutRangeMin:float=0
      OutRangeMax:float=0
     ) -> (ReturnValue:float) {is=title K2Node_CallFunction}
     B=50
    ) {is=title K2Node_CallFunction})
   n8 = Reroute Node()
   TryGetPawnOwner()
  ]}
]
variables [
 crouch_alpha {type=number}
 isWearingBackbling {type=boolean}
 isWearingCape {type=boolean}
]
